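# Animation Blueprint for the medium Avocado Eater Bird, a parasite creature that rides on a player and reacts to their actions.
# It copies the parent mesh pose, layers a state machine from alien_root, then applies RigidBody physics and inertialization.
# A conduit routes locomotion, crouch, ADS, jump and vehicle states using flags derived each update.
# The update also starts an async cosmetic-part wait that sets fit morph targets by matching mesh names.
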AnimBlueprint /MotherGameplay/Items/AvacadoEaterBird/Animation/AvocadoEater_Medium/AvocadoEater_Medium_AnimBP
target_skeleton=/MotherGameplay/Items/AvacadoEaterBird/Meshes/Avocado_Eater_Bird_MED/Avocado_Eater_Bird_Mammal_Skeleton
graphs [
 AnimGraph {nodes [
   n1 = Copy Pose From Mesh() {bCopyCurves=true, bUseAttachedParent=true}
   AnimGraphNode_StateMachine_0 = AvocadoEater_Med_Statemachine() {is=title AnimGraphNode_StateMachine}
   n2 = RigidBody(ComponentPose=Local To Component(LocalPose=Layered blend per bone(BasePose=@n1, BlendPoses_0=@AnimGraphNode_StateMachine_0) {layerSetup [{branchFilters [{boneName=alien_root}]}]})) {
    bOverrideWorldGravity=true
    overridePhysicsAsset=/MotherGameplay/Items/AvacadoEaterBird/Meshes/Avocado_Eater_Bird_MED/Avocado_Eater_Bird_Mammal_Physics
    simulationSpace=BaseBoneSpace
    alphaBoolBlend {blendInTime=0.35, blendOption=Linear, blendOutTime=0.35}
    baseBoneRef {boneName=root}
    componentAppliedLinearAccClamp=<2000 2000 2000>
    componentLinearAccScale=<0.3 0.3 0.35>
    componentLinearVelScale=<0.3 0.3 0.35>
    overrideWorldGravity {z=-980}
   }
   Output Pose(Result=Inertialization!(Source:Pose Link Structure=Component To Local(ComponentPose=@n2)) -> (Pose:Pose Link Structure) {is=title AnimGraphNode_Inertialization}) {layerGroup=None}
  ]}
 AvocadoEater_Med_Statemachine {
  entry=DirtyHackState
  path=AnimGraph.AnimGraphNode_StateMachine_0.AvocadoEater_Med_Statemachine
  ids=id
  nodes [
   AnimStateConduitNode_0 = Conduit()
   AnimStateNode_0 = DirtyHackState() {is=title AnimStateNode}
   AnimStateNode_1 = Locomotion() {is=title AnimStateNode}
   AnimStateNode_2 = Crouch() {is=title AnimStateNode}
   AnimStateNode_3 = ADS() {is=title AnimStateNode}
   AnimStateNode_4 = JumpStart() {is=title AnimStateNode}
   AnimStateNode_5 = JumpLoop() {is=title AnimStateNode}
   AnimStateNode_6 = JumpLand() {is=title AnimStateNode}
   AnimStateNode_7 = InVehicle() {is=title AnimStateNode}
   AnimStateTransitionNode_0 = AnimStateTransitionNode()
   AnimStateTransitionNode_1 = AnimStateTransitionNode() {logicType=TLT_Inertialization, priorityOrder=100}
   AnimStateTransitionNode_10 = AnimStateTransitionNode() {crossfadeDuration=0.1, logicType=TLT_Inertialization}
   AnimStateTransitionNode_11 = AnimStateTransitionNode()
   AnimStateTransitionNode_12 = AnimStateTransitionNode()
   AnimStateTransitionNode_13 = AnimStateTransitionNode()
   AnimStateTransitionNode_14 = AnimStateTransitionNode() {crossfadeDuration=0.3, logicType=TLT_Inertialization, priorityOrder=100}
   AnimStateTransitionNode_15 = AnimStateTransitionNode() {logicType=TLT_Inertialization}
   AnimStateTransitionNode_16 = AnimStateTransitionNode() {logicType=TLT_Inertialization}
   AnimStateTransitionNode_2 = AnimStateTransitionNode()
   AnimStateTransitionNode_3 = AnimStateTransitionNode() {logicType=TLT_Inertialization}
   AnimStateTransitionNode_4 = AnimStateTransitionNode() {crossfadeDuration=0.35, logicType=TLT_Inertialization, priorityOrder=2}
   AnimStateTransitionNode_5 = AnimStateTransitionNode()
   AnimStateTransitionNode_6 = AnimStateTransitionNode()
   AnimStateTransitionNode_7 = AnimStateTransitionNode() {logicType=TLT_Inertialization, priorityOrder=4}
   AnimStateTransitionNode_8 = AnimStateTransitionNode() {logicType=TLT_Inertialization, priorityOrder=3}
   AnimStateTransitionNode_9 = AnimStateTransitionNode() {bAutomaticRuleBasedOnSequencePlayerInState=true, crossfadeDuration=0.3, logicType=TLT_Inertialization}
  ]
  links [
   {
    dst=Conduit
    rule=AnimGraph.AnimGraphNode_StateMachine_0.AvocadoEater_Med_Statemachine.AnimStateTransitionNode_5.Transition
    src=ADS
    via=AnimStateTransitionNode_5
   }
   {
    dst=InVehicle
    rule=AnimGraph.AnimGraphNode_StateMachine_0.AvocadoEater_Med_Statemachine.AnimStateTransitionNode_16.Transition
    src=ADS
    via=AnimStateTransitionNode_16
   }
   {
    dst=ADS
    rule=AnimGraph.AnimGraphNode_StateMachine_0.AvocadoEater_Med_Statemachine.AnimStateTransitionNode_4.Transition
    src=Conduit
    via=AnimStateTransitionNode_4
   }
   {
    dst=Crouch
    rule=AnimGraph.AnimGraphNode_StateMachine_0.AvocadoEater_Med_Statemachine.AnimStateTransitionNode_3.Transition
    src=Conduit
    via=AnimStateTransitionNode_3
   }
   {
    dst=InVehicle
    rule=AnimGraph.AnimGraphNode_StateMachine_0.AvocadoEater_Med_Statemachine.AnimStateTransitionNode_8.Transition
    src=Conduit
    via=AnimStateTransitionNode_8
   }
   {
    dst=JumpLoop
    rule=AnimGraph.AnimGraphNode_StateMachine_0.AvocadoEater_Med_Statemachine.AnimStateTransitionNode_13.Transition
    src=Conduit
    via=AnimStateTransitionNode_13
   }
   {
    dst=JumpStart
    rule=AnimGraph.AnimGraphNode_StateMachine_0.AvocadoEater_Med_Statemachine.AnimStateTransitionNode_7.Transition
    src=Conduit
    via=AnimStateTransitionNode_7
   }
   {
    dst=Locomotion
    rule=AnimGraph.AnimGraphNode_StateMachine_0.AvocadoEater_Med_Statemachine.AnimStateTransitionNode_1.Transition
    src=Conduit
    via=AnimStateTransitionNode_1
   }
   {
    dst=Conduit
    rule=AnimGraph.AnimGraphNode_StateMachine_0.AvocadoEater_Med_Statemachine.AnimStateTransitionNode_6.Transition
    src=Crouch
    via=AnimStateTransitionNode_6
   }
   {
    dst=Conduit
    rule=AnimGraph.AnimGraphNode_StateMachine_0.AvocadoEater_Med_Statemachine.AnimStateTransitionNode_0.Transition
    src=DirtyHackState
    via=AnimStateTransitionNode_0
   }
   {
    dst=ADS
    rule=AnimGraph.AnimGraphNode_StateMachine_0.AvocadoEater_Med_Statemachine.AnimStateTransitionNode_15.Transition
    src=InVehicle
    via=AnimStateTransitionNode_15
   }
   {
    dst=Conduit
    rule=AnimGraph.AnimGraphNode_StateMachine_0.AvocadoEater_Med_Statemachine.AnimStateTransitionNode_12.Transition
    src=InVehicle
    via=AnimStateTransitionNode_12
   }
   {
    dst=Conduit
    rule=AnimGraph.AnimGraphNode_StateMachine_0.AvocadoEater_Med_Statemachine.AnimStateTransitionNode_11.Transition
    src=JumpLand
    via=AnimStateTransitionNode_11
   }
   {
    dst=JumpLand
    rule=AnimGraph.AnimGraphNode_StateMachine_0.AvocadoEater_Med_Statemachine.AnimStateTransitionNode_10.Transition
    src=JumpLoop
    via=AnimStateTransitionNode_10
   }
   {
    dst=JumpStart
    rule=AnimGraph.AnimGraphNode_StateMachine_0.AvocadoEater_Med_Statemachine.AnimStateTransitionNode_14.Transition
    src=JumpLoop
    via=AnimStateTransitionNode_14
   }
   {
    dst=JumpLoop
    rule=AnimGraph.AnimGraphNode_StateMachine_0.AvocadoEater_Med_Statemachine.AnimStateTransitionNode_9.Transition
    src=JumpStart
    via=AnimStateTransitionNode_9
   }
   {
    dst=Conduit
    rule=AnimGraph.AnimGraphNode_StateMachine_0.AvocadoEater_Med_Statemachine.AnimStateTransitionNode_2.Transition
    src=Locomotion
    via=AnimStateTransitionNode_2
   }
  ]
 }
 Conduit {
  path=AnimGraph.AnimGraphNode_StateMachine_0.AvocadoEater_Med_Statemachine.AnimStateConduitNode_0.Conduit
  nodes [Result(bCanEnterTransition=true)]
 }
 DirtyHackState {
  path=AnimGraph.AnimGraphNode_StateMachine_0.AvocadoEater_Med_Statemachine.AnimStateNode_0.DirtyHackState
  nodes [Output Animation Pose(Result=Local Space Ref Pose())]
 }
 Locomotion {
  path=AnimGraph.AnimGraphNode_StateMachine_0.AvocadoEater_Med_Statemachine.AnimStateNode_1.Locomotion
  nodes [
   n1 = Random Sequence Player!() -> (Pose:Pose Link Structure) {
    is=title AnimGraphNode_RandomPlayer
    entries [
     {
      chanceToPlay=0.7
      maxLoopCount=0
      maxPlayRate=1
      minLoopCount=0
      minPlayRate=1
      sequence=&/AvocadoEater_Med_Idle
      blendIn {blendOption=Linear, blendTime=0.2, customCurve=null}
     }
     {
      chanceToPlay=0.3
      maxLoopCount=0
      maxPlayRate=1
      minLoopCount=0
      minPlayRate=1
      sequence=&/AvocadoEater_Med_Idle_02
      blendIn {blendOption=Linear, blendTime=0.2, customCurve=null}
     }
    ]
   }
   Output Animation Pose(Result=Blend Poses by bool(
     BlendPose_0=Sequence Player 'AvocadoEater_Med_Sprint'() {sequence=&/AvocadoEater_Med_Sprint}
     BlendPose_1=@n1
     BlendTime_0=0.6
     BlendTime_1=0.4
     bActiveValue=Get bIsSprinting()
    ) {blendType=Linear, transitionType=Inertialization})
  ]
 }
 Crouch {
  path=AnimGraph.AnimGraphNode_StateMachine_0.AvocadoEater_Med_Statemachine.AnimStateNode_2.Crouch
  nodes [Output Animation Pose(Result=Sequence Player 'AvocadoEater_Med_Crouch'() {sequence=&/AvocadoEater_Med_Crouch})]
 }
 ADS {
  path=AnimGraph.AnimGraphNode_StateMachine_0.AvocadoEater_Med_Statemachine.AnimStateNode_3.ADS
  nodes [Output Animation Pose(Result=Sequence Player 'AvocadoEater_Med_ADS'() {sequence=&/AvocadoEater_Med_ADS})]
 }
 JumpStart {
  path=AnimGraph.AnimGraphNode_StateMachine_0.AvocadoEater_Med_Statemachine.AnimStateNode_4.JumpStart
  nodes [Output Animation Pose(Result=Sequence Player 'AvocadoEater_Med_Jump_Start'() {bLoopAnimation=false, sequence=&/AvocadoEater_Med_Jump_Start})]
 }
 JumpLoop {
  path=AnimGraph.AnimGraphNode_StateMachine_0.AvocadoEater_Med_Statemachine.AnimStateNode_5.JumpLoop
  nodes [Output Animation Pose(Result=Sequence Player 'AvocadoEater_Med_Jump_Loop'() {sequence=&/AvocadoEater_Med_Jump_Loop})]
 }
 JumpLand {
  path=AnimGraph.AnimGraphNode_StateMachine_0.AvocadoEater_Med_Statemachine.AnimStateNode_6.JumpLand
  nodes [Output Animation Pose(Result=Sequence Player 'AvocadoEater_Med_Jump_End'() {bLoopAnimation=false, sequence=&/AvocadoEater_Med_Jump_End})]
 }
 InVehicle {
  path=AnimGraph.AnimGraphNode_StateMachine_0.AvocadoEater_Med_Statemachine.AnimStateNode_7.InVehicle
  nodes [Output Animation Pose(Result=Sequence Player 'AvocadoEater_Med_VehicleIdle'() {sequence=&/AvocadoEater_Med_VehicleIdle})]
 }
 Transition {
  path=AnimGraph.AnimGraphNode_StateMachine_0.AvocadoEater_Med_Statemachine.AnimStateTransitionNode_0.Transition
  nodes [Result(bCanEnterTransition=true)]
 }
 Transition {
  path=AnimGraph.AnimGraphNode_StateMachine_0.AvocadoEater_Med_Statemachine.AnimStateTransitionNode_1.Transition
  nodes [Result(bCanEnterTransition=true)]
 }
 Transition {
  path=AnimGraph.AnimGraphNode_StateMachine_0.AvocadoEater_Med_Statemachine.AnimStateTransitionNode_10.Transition
  nodes [Result(bCanEnterTransition=Get IsLanding() -> (IsLanding:bool))]
 }
 Transition {
  path=AnimGraph.AnimGraphNode_StateMachine_0.AvocadoEater_Med_Statemachine.AnimStateTransitionNode_11.Transition
  nodes [Result(bCanEnterTransition=Get Interruptible() -> (Interruptible:bool))]
 }
 Transition {
  path=AnimGraph.AnimGraphNode_StateMachine_0.AvocadoEater_Med_Statemachine.AnimStateTransitionNode_12.Transition
  nodes [Result(bCanEnterTransition=NOT Boolean(A=Get IsInClosedInVehicle() -> (IsInClosedInVehicle:bool)))]
 }
 Transition {
  path=AnimGraph.AnimGraphNode_StateMachine_0.AvocadoEater_Med_Statemachine.AnimStateTransitionNode_13.Transition
  nodes [Result(bCanEnterTransition=Get bIsFalling())]
 }
 Transition {
  path=AnimGraph.AnimGraphNode_StateMachine_0.AvocadoEater_Med_Statemachine.AnimStateTransitionNode_14.Transition
  nodes [Result(bCanEnterTransition=Get IsGoopDoubleJump() -> (IsGoopDoubleJump:bool))]
 }
 Transition {
  path=AnimGraph.AnimGraphNode_StateMachine_0.AvocadoEater_Med_Statemachine.AnimStateTransitionNode_15.Transition
  nodes [Result(bCanEnterTransition=Get bIsTargeting())]
 }
 Transition {
  path=AnimGraph.AnimGraphNode_StateMachine_0.AvocadoEater_Med_Statemachine.AnimStateTransitionNode_16.Transition
  nodes [Result(bCanEnterTransition=NOT Boolean(A=Get bIsTargeting()))]
 }
 Transition {
  path=AnimGraph.AnimGraphNode_StateMachine_0.AvocadoEater_Med_Statemachine.AnimStateTransitionNode_2.Transition
  nodes [Result(bCanEnterTransition=true)]
 }
 Transition {
  path=AnimGraph.AnimGraphNode_StateMachine_0.AvocadoEater_Med_Statemachine.AnimStateTransitionNode_3.Transition
  nodes [Result(bCanEnterTransition=Get bIsCrouching())]
 }
 Transition {
  path=AnimGraph.AnimGraphNode_StateMachine_0.AvocadoEater_Med_Statemachine.AnimStateTransitionNode_4.Transition
  nodes [Result(bCanEnterTransition=Get IsParasiteADS() -> (IsParasiteADS:bool))]
 }
 Transition {
  path=AnimGraph.AnimGraphNode_StateMachine_0.AvocadoEater_Med_Statemachine.AnimStateTransitionNode_5.Transition
  nodes [Result(bCanEnterTransition=NOT Boolean(A=Get IsParasiteADS()))]
 }
 Transition {
  path=AnimGraph.AnimGraphNode_StateMachine_0.AvocadoEater_Med_Statemachine.AnimStateTransitionNode_6.Transition
  nodes [Result(bCanEnterTransition=NOT Boolean(A=Get bIsCrouching()))]
 }
 Transition {
  path=AnimGraph.AnimGraphNode_StateMachine_0.AvocadoEater_Med_Statemachine.AnimStateTransitionNode_7.Transition
  nodes [Result(bCanEnterTransition=Get bIsJumping())]
 }
 Transition {
  path=AnimGraph.AnimGraphNode_StateMachine_0.AvocadoEater_Med_Statemachine.AnimStateTransitionNode_8.Transition
  nodes [Result(bCanEnterTransition=Get IsInClosedInVehicle())]
 }
 Transition {
  path=AnimGraph.AnimGraphNode_StateMachine_0.AvocadoEater_Med_Statemachine.AnimStateTransitionNode_9.Transition
  nodes [Result(bCanEnterTransition=true)]
 }
 EventGraph {nodes [
   Event BlueprintUpdateAnimation()
   > n1 = Sequence!(execute) -> (then_0, then_1, then_2:Exec, then_3:Exec, then_4:Exec, then_5:Exec, then_6:Exec)
   > n2 = Branch(Condition="float >= float"(A=GetCurveValue(CurveName=TurnOffAlpha), B=0.9) {is=title K2Node_CallFunction})
   > Set TurnOffAlpha(TurnOffAlpha:bool=false) -> (Output_Get:bool)
   Set TurnOffAlpha(execute=@n2.else, TurnOffAlpha=true)
   n3 = Branch(execute=@n1.then_1, Condition="float >= float"(A=GetCurveValue(CurveName=Interruptible), B=0.98))
   > Set Interruptible(Interruptible:bool=true) -> (Output_Get:bool)
   Set Interruptible(execute=@n3.else, Interruptible=false)
   n4 = Get bIsJumping()
   n5 = Get bIsFalling()
   n6 = Branch(execute=@n1.then_2, Condition=AND Boolean(A=NOT Boolean(A=@n5), B=NOT Boolean(A=@n4)))
   > Set IsLanding(IsLanding:bool=true) -> (Output_Get:bool)
   Set IsLanding(execute=@n6.else, IsLanding=false)
   n7 = "Equal (Enum)"(
    A:ESlotEnvironmentExposure Enum=Get VehicleSlotExposure() -> (VehicleSlotExposure:ESlotEnvironmentExposure Enum=Unknown)
    B:ESlotEnvironmentExposure Enum=Exposed
   )
   n8 = Get bIsCrouching()
   n9 = Branch(
    execute=@n1.then_3
    Condition=AND Boolean!(A=Get bIsInVehicle() -> (bIsInVehicle:bool), B=NOT Boolean(A=@n8), C:bool=NOT Boolean(A=@n7)) -> (ReturnValue)
   )
   > Set IsInClosedInVehicle(IsInClosedInVehicle:bool=true) -> (Output_Get:bool)
   Set IsInClosedInVehicle(execute=@n9.else, IsInClosedInVehicle=false)
   n10 = Branch(execute=@n1.then_4, Condition=AND Boolean!(A=Get bIsJumping(), B=Get bIsInGoop() -> (bIsInGoop:bool)) -> (ReturnValue))
   > Set IsGoopDoubleJump(IsGoopDoubleJump:bool=true) -> (Output_Get:bool)
   Set IsGoopDoubleJump(execute=@n10.else, IsGoopDoubleJump=false)
   n11 = Get bIsCrouching()
   n12 = Branch(execute=@n1.then_5, Condition=AND Boolean(A=Get bIsTargeting(), B=NOT Boolean(A=@n11)))
   > Set IsParasiteADS(IsParasiteADS:bool=true) -> (Output_Get:bool)
   Set IsParasiteADS(execute=@n12.else, IsParasiteADS=false)
   n13 = Cast To FortPlayerPawn(execute=@n1.then_6, Object=TryGetPawnOwner())
   > n14 = WaitForCosmeticSpecificPartApplied!(
    execute:Exec
    FortPlayerPawn:Fort Player Pawn Object Reference=@n13
    bSkipNullMesh:bool=true
    PartType:EFortCustomPartType Enum
    AnyObjects:Array of Object References
    AnyStrings:Array of Strings
   ) -> (
    then:Exec
    OnPartApplied:Exec
    Pawn:Fort Player Pawn Object Reference
    Part:Custom Character Part Object Reference
    MeshPart:Skeletal Mesh Component Object Reference
    PartType:EFortCustomPartType Enum
    AnyObjects:Array of Object References
    AnyStrings:Array of Strings
   ) {is=title K2Node_AsyncAction}
   n15 = GetDisplayName(Object=@n14.MeshPart)
   SetMorphTarget!(
    execute=@n14.OnPartApplied
    self:Anim Instance Object Reference
    MorphTargetName:Name=Comet_Shape
    Value:float="To Float (Boolean)"(InBool=MatchesWildcard(SourceString=@n15, Wildcard=*M_LRG_Comet*))
   ) -> (then)
   > SetMorphTarget(MorphTargetName=Chuggus_Shape, Value="To Float (Boolean)"(InBool=MatchesWildcard(SourceString=@n15, Wildcard=*M_LRG_Big_Chuggus*)))
   > SetMorphTarget(MorphTargetName=Meteor_Man_Shape, Value="To Float (Boolean)"(InBool=MatchesWildcard(SourceString=@n15, Wildcard=*M_LRG_Meteor_Man_Remix*)))
   > SetMorphTarget(MorphTargetName=Nana_Split_Shape, Value="To Float (Boolean)"(InBool=MatchesWildcard(SourceString=@n15, Wildcard=*M_MED_Nana_Split*)))
   > SetMorphTarget(MorphTargetName=Banana_Shape, Value="To Float (Boolean)"(InBool=MatchesWildcard(SourceString=@n15, Wildcard=*M_MED_Banana*)))
   > SetMorphTarget(MorphTargetName=Nosh_Shape, Value="To Float (Boolean)"(InBool=MatchesWildcard(SourceString=@n15, Wildcard=*M_MED_Nosh*)))
   > SetMorphTarget(MorphTargetName=Henchman_Tough_Shape, Value="To Float (Boolean)"(InBool=MatchesWildcard(SourceString=@n15, Wildcard=*M_LRG_Henchman_Tough*)))
   > SetMorphTarget(MorphTargetName=Chuggus_Shape, Value="To Float (Boolean)"(InBool=MatchesWildcard(SourceString=@n15, Wildcard=*M_MED_Shock_Wave_Commando*)))
   > SetMorphTarget(MorphTargetName=Corny_Shape, Value="To Float (Boolean)"(InBool=MatchesWildcard(SourceString=@n15, Wildcard=*M_MED_Corny*)))
   > SetMorphTarget(
    MorphTargetName=Chicken_Warrior_ShapeShape
    Value="To Float (Boolean)"(InBool=MatchesWildcard(SourceString=@n15, Wildcard=*M_LRG_Chicken_Warrior*))
   )
   > SetMorphTarget(MorphTargetName=Invader_Shape, Value="To Float (Boolean)"(InBool=MatchesWildcard(SourceString=@n15, Wildcard=*M_LRG_Invader*)))
   > SetMorphTarget(MorphTargetName=Cattus_Shape, Value="To Float (Boolean)"(InBool=MatchesWildcard(SourceString=@n15, Wildcard=*M_LRG_Cattus*)))
   > SetMorphTarget(MorphTargetName=Henchman_Tough_Shape, Value="To Float (Boolean)"(InBool=MatchesWildcard(SourceString=@n15, Wildcard=*M_LRG_Vertigo*)))
   > SetMorphTarget(
    MorphTargetName=Henchman_Tough_Shape
    Value="To Float (Boolean)"(InBool=MatchesWildcard(SourceString=@n15, Wildcard=*M_LRG_Henchman_Summer*))
   )
   > SetMorphTarget(MorphTargetName=Ruckus_Mini_Shape, Value="To Float (Boolean)"(InBool=MatchesWildcard(SourceString=@n15, Wildcard=*M_MED_Ruckus_Mini*)))
   TryGetPawnOwner()
  ]}
]
variables [
 Interruptible {type=boolean}
 IsGoopDoubleJump {type=boolean}
 IsInClosedInVehicle {type=boolean}
 IsLanding {type=boolean}
 IsParasiteADS {type=boolean}
 StartJump {type=boolean}
 TurnOffAlpha {type=boolean}
]
